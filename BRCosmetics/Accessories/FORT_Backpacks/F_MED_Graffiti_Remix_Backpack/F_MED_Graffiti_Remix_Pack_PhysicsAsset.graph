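# Physics asset for the Graffiti Remix backpack cosmetic, defining collision and joint limits for its dangling canisters and straps.
# Capsules cover the pack body and three canisters, while boxes cover the strap and its left and right ends.
# Every Dyn_ bone constrains directly to pack_mid rather than chaining.
# Canister A swings widely, while canisters B and C are held to tight swing and twist limits.
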PhysicsAsset /BRCosmetics/Accessories/FORT_Backpacks/F_MED_Graffiti_Remix_Backpack/F_MED_Graffiti_Remix_Pack_PhysicsAsset
bodies {
 Dyn_canister_a [{shapeType=Capsule, length=8.4257, radius=3.6719, center=<2.4459 -2.7994 -0.8869>, rotation=<rot 0 -87.5553 -85.1017>}]
 Dyn_canister_b [{shapeType=Capsule, length=9.2363, radius=2.9261, center=<-1.3362 -2.4804 -0.7164>, rotation=<rot 90 0 0>}]
 Dyn_canister_c [{shapeType=Capsule, length=9.3668, radius=3.3332, center=<-1.3386 -2.7906 -0.931>, rotation=<rot 90 0 0>}]
 Dyn_strap [{shapeType=Box, extentX=3.97, extentY=4.9795, extentZ=1.75, center=<-1.0191 -0.0001 0>, rotation=<rot 90 0 0>}]
 Dyn_strapend_l [{shapeType=Box, extentX=1.7276, extentY=1.1564, extentZ=4.3511, center=<1.9653 -0.0168 -0.0758>}]
 Dyn_strapend_r [{shapeType=Box, extentX=3.1567, extentY=0.9289, extentZ=4.0087, center=<1.4 -0.0426 -0.0916>}]
 pack_mid [{shapeType=Capsule, length=9.625, radius=5.625}]
}
constraints [
 {bone1Name=Dyn_strap, bone2Name=pack_mid, swing1LimitDegrees=45, swing2LimitDegrees=45, twistLimitDegrees=45}
 {bone1Name=Dyn_canister_a, bone2Name=pack_mid, swing1LimitDegrees=53.8571, swing2LimitDegrees=79.5714, twistLimitDegrees=12}
 {bone1Name=Dyn_canister_b, bone2Name=pack_mid, swing1LimitDegrees=15, swing2LimitDegrees=15, twistLimitDegrees=3}
 {bone1Name=Dyn_canister_c, bone2Name=pack_mid, swing1LimitDegrees=15, swing2LimitDegrees=15, twistLimitDegrees=3}
 {bone1Name=Dyn_strapend_l, bone2Name=pack_mid, swing1LimitDegrees=35, swing2LimitDegrees=15, twistLimitDegrees=10}
 {bone1Name=Dyn_strapend_r, bone2Name=pack_mid, swing1LimitDegrees=45, swing2LimitDegrees=15, twistLimitDegrees=10}
]
